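# UI material for a Discovery browser billboard live event, drawing an animated, tintable glow (red by default) with alpha-composite blending.
# Radial coordinates feed parameterized tiling and panning of a blurred noise texture, sampled with derivatives from the unwarped UI UVs.
# An offset, smoothstepped circle SDF confines the noise, which drives color intensity and is clipped by a rounded-corner mask.
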Material /DiscoveryBrowser/Rows/Billboard/MaterialInstances/M_UI_LiveEvent
graph {
 properties {
  bAutomaticallySetUsageInEditor=false
  bUsedWithStaticMesh=false
  blendMode=BLEND_AlphaComposite
  materialDomain=MD_UI
  usedShadingModels=Default Lit
 }
 nodes [
  n1 = Multiply(A=MaterialFunctionCall() {materialFunction=&GetUserInterfaceUV}.Normalized UV) {constB=0.1}
  n2 = MaterialFunctionCall(Float2=MaterialFunctionCall!(Vector or UVs=ConstantBiasScale(None=@n1) {bias=0.5, scale=2}) {materialFunction=/FortUILibrary/MaterialFunctions/Basics/MF_UI_CheaperVectorToRadial}.Radial Coordinates) {materialFunction=&BreakOutFloat2Components}
  n3 = Multiply(A=@n2.R, B=ScalarParameter() {defaultValue=1, parameterName=Perpendicular Repitition})
  n4 = Multiply(A=@n2.G, B=ScalarParameter() {defaultValue=2, parameterName=Parallel Repitition})
  n5 = ScalarParameter() {defaultValue=-0.05, parameterName=Perpendicular Speed}
  n6 = ScalarParameter() {defaultValue=-0.1, parameterName=Parallel Speed}
  n7 = MaterialFunctionCall() {materialFunction=&GetUserInterfaceUV}
  n8 = TextureSample!(
   UVs=Panner(Coordinate=AppendVector(A=@n3, B=@n4), Time=Time(), Speed=AppendVector(A=@n5, B=@n6))
   Tex
   "DDX(UVs)"=DDX(Value=@n7.Normalized UV)
   "DDY(UVs)"=DDY(Value=@n7.Normalized UV)
   Apply View MipBias
  ) {mipValueMode=TMVM_Derivative, samplerType=SAMPLERTYPE_Grayscale, texture=/FortUILibrary/UIKit/Buttons/Textures/LowResBlurredNoise}
  n9 = MaterialFunctionCall() {materialFunction=&GetUserInterfaceUV}
  n10 = Constant() {r=-1}
  n11 = Constant() {r=-0.6}
  n12 = MaterialFunctionCall(UVs=Subtract(
    A=MaterialFunctionCall!(ScreenUV=@n9.Normalized UV, TextureSize, ScalePivotPoint) {materialFunction=/FortUILibrary/MaterialFunctions/Basics/MF_UI_ScreenSpaceUVScaling}.Scaled to Fit
    B=AppendVector(A=@n10, B=@n11)
   ) {constB=-0.5}) {materialFunction=&MF_UI_Circle}
  n13 = Multiply(A=@n8.R, B=SmoothStep(Value=OneMinus(None=@n12.SDF)))
  n14 = Constant() {r=1}
  n15 = ScalarParameter() {defaultValue=10, parameterName=Corner Radius}
  Multiply(
   A=LinearInterpolate(B=VectorParameter() {parameterName=Color, defaultValue=<rgba 1 0.0097 0 1>}.RGB, Alpha=@n13)
   B=Multiply(
    A=@n13
    B=MaterialFunctionCall(Size=@n14, Corner Radius=@n15, Radius in Pixels=@n14) {materialFunction=&MF_UI_CoreUI_RoundedMask}.Shape Mask
   )
  )
 ]
}
